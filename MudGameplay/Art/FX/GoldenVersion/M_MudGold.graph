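# Animated molten gold material for mud FX (the golden version), usable on Niagara sprites and meshes and on spline meshes.
# Scrolling noise plus a sine-shaped Fresnel term drives a stepped dark-to-highlight gold ramp with metallic shading and a flattenable normal map.
# A second noise mask blends in emissive lava, whose glow colors follow a mask built from two panning lava textures warped by a third.
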Material /MudGameplay/Art/FX/GoldenVersion/M_MudGold
graph {
 properties {
  allowTranslucentCustomDepthWrites=true
  bUseMaterialAttributes=true
  bUsedWithNiagaraMeshParticles=true
  bUsedWithNiagaraSprites=true
  bUsedWithSplineMeshes=true
  blendMode=BLEND_Masked
  pixelDepthOffsetMode=PDOM_Legacy
  usedShadingModels=Default Lit
 }
 nodes [
  n1 = VectorParameter() {group=GoldenLava, parameterName=GoldColorDark, sortPriority=12, defaultValue=<rgba 0.2292 0.1142 0 1>}
  n2 = VectorParameter() {group=GoldenLava, parameterName=GoldenHighlight, sortPriority=15, defaultValue=<rgba 0.276 0.2387 0.1294 1>}
  n3 = Fresnel() {baseReflectFraction=0, exponent=2}
  n4 = Divide(A=Sine(None=Multiply(A=@n3) {constB=2}))
  n5 = NamedRerouteUsage()
  n6 = Divide(A=Add(
    A=Add(A=@n4) {constB=0.5}
    B=TextureSample(UVs=@n5) {samplerType=SAMPLERTYPE_LinearColor, texture=&/T_MoltenGold_Noise}.R
   ))
  n7 = LinearInterpolate(
   A=VectorParameter() {group=GoldenLava, parameterName=GoldColor, sortPriority=11, defaultValue=<rgba 1 0.75 0 1>}.RGB
   B=@n2.RGB
   Alpha=SmoothStep(Value=@n6) {constMin=0.95}
  )
  n8 = SmoothStep(Value=@n6) {constMax=0.7, constMin=0.5}
  n9 = NamedRerouteUsage()
  n10 = SetMaterialAttributes!(
   MaterialAttributes
   Base Color=LinearInterpolate(
    A=VectorParameter() {group=GoldenLava, parameterName=GoldColorDarker, sortPriority=12, defaultValue=<rgba 0 0 0 1>}.RGB
    B=LinearInterpolate(A=@n1.RGB, B=@n7, Alpha=@n8)
    Alpha=SmoothStep(Value=@n6) {constMax=0.4}
   )
   Metallic=Constant() {r=0.75}
   Specular=Constant() {r=1}
   Roughness=Constant() {r=0.3}
   Normal=@n9
  ) {attributeSetTypes [
    #1
    #2
    #3
    #4
    #5
   ]}
  n11 = VectorParameter() {parameterName=LavaGlowEdgeDark, defaultValue=<rgba 0.0104 0.0004 0 1>}
  n12 = VectorParameter() {parameterName=LavalGlowEdgeBright, defaultValue=<rgba 20 0.8 0 1>}
  n13 = NamedRerouteUsage()
  n14 = Constant()
  n15 = SetMaterialAttributes!(
   MaterialAttributes
   Emissive Color=Multiply(A=LinearInterpolate(A=@n11.RGB, B=@n12.RGB, Alpha=@n13)) {constB=1.5}
   Base Color=@n14
   Metallic=@n14
   Specular=@n14
   Roughness=Constant() {r=1}
   Normal
   World Position Offset
  ) {attributeSetTypes [
    #6
    #1
    #2
    #3
    #4
    #5
    #7
   ]}
  n16 = MaterialFunctionCall(Float2=NamedRerouteUsage()) {materialFunction=&BreakOutFloat2Components}
  n17 = TextureSample(UVs=AppendVector(A=@n16.G, B=@n16.G)) {samplerType=SAMPLERTYPE_LinearColor, texture=&T_MoltenGold_Noise}
  BlendMaterialAttributes(
   A=@n10
   B=@n15
   Alpha=SmoothStep(
    Min=ScalarParameter() {defaultValue=0.8, group=GoldenLava, parameterName=LavaProportionMin, sortPriority=81}
    Max=ScalarParameter() {defaultValue=1, group=GoldenLava, parameterName=LavaProportionMax, sortPriority=82}
    Value=OneMinus(None=@n17.R)
   ) {constMin=0.8}
  )
  Constant3Vector() {constant=<rgba 1 1 0 0>}
  Divide(A=ComponentMask(None=WorldPosition().XYZ) {b=true, r=true}) {constB=20}
  n18 = Subtract(
   A=WorldPosition() {worldPositionShaderOffset=WPT_ExcludeAllShaderOffsets}.XYZ
   B=MaterialFunctionCall() {materialFunction=&ObjectPivotPoint}.Object Pivot Location
  )
  n19 = ScalarParameter() {defaultValue=0.01, group=GoldenLava, parameterName=Scale}
  Multiply(A=Add(A=@n9, B=Multiply(A=@n18, B=@n19) {constB=0})) {constB=0.5}
  n20 = Time()
  n21 = Constant2Vector() {g=0.1}
  n22 = ComponentMask(None=Add(A=TextureCoordinate(), B=Multiply(A=@n20, B=@n21))) {g=true, r=true}
  n23 = ScalarParameter() {defaultValue=5, parameterName=LavaEdge_Tiling1}
  n24 = ScalarParameter() {defaultValue=1, parameterName=LavaEdgeTiling3}
  n25 = VectorParameter() {parameterName=PanSpeedEdgeGlow3, defaultValue=<rgba 0.005 -0.02 0 1>}
  n26 = TextureObject() {samplerType=SAMPLERTYPE_Masks, texture=/Game/Environments/Sets/Vulcano/Textures/Lava_M}
  n27 = TextureSample(UVs=Panner(Coordinate=Divide(A=@n22, B=@n24), Speed=ComponentMask(None=@n25.RGB) {g=true, r=true}), Tex=@n26) {samplerType=SAMPLERTYPE_Masks}
  n28 = ScalarParameter() {defaultValue=0.2, parameterName=EdgeBlendWarp}
  n29 = Add(
   A=Panner(Coordinate=Divide(A=@n22, B=@n23), Speed=ScalarParameter() {defaultValue=0.02, parameterName=LavaEdge_Speed_Glow})
   B=Multiply(A=@n27.R, B=@n28)
  )
  n30 = Divide(A=@n22, B=ScalarParameter() {defaultValue=10, parameterName=LavaEdge_Tiling2})
  n31 = Add(
   A=Panner(
    Coordinate=MaterialFunctionCall(XY=@n30) {materialFunction=&Swizzle}.YX
    Speed=ScalarParameter() {defaultValue=-0.015, parameterName=PanSpeedEdgeGlow2}
   )
   B=Multiply(A=@n27.G, B=@n28) {constB=0.2}
  )
  n32 = Add(A=Add(
    A=TextureSample(UVs=@n29, Tex=@n26) {samplerType=SAMPLERTYPE_Masks, texture=/Game/Environments/Sets/Vulcano/Textures/LavaGround_M}.G
    B=TextureSample(UVs=@n31, Tex=@n26) {samplerType=SAMPLERTYPE_Masks, texture=&LavaGround_M}.G
   )) {constB=-0.6}
  NamedRerouteDeclaration(None=LinearInterpolate(Alpha=Saturate(None=@n32)) {constA=0.5}) {name=LavaMask, nodeColor=<rgba 0 0.8828 1 1>}
  n33 = Multiply(
   A=ComponentMask(None=TextureCoordinate()) {g=true, r=true}
   B=AppendVector(
    A=ScalarParameter() {defaultValue=1, group=GoldenLava, parameterName=NormalTillingX, sortPriority=33}
    B=ScalarParameter() {defaultValue=1, group=GoldenLava, parameterName=NormalTillingY, sortPriority=34}
   )
  )
  n34 = Multiply(A=AppendVector(A=Constant(), B=ScalarParameter() {defaultValue=0.05, group=GoldenLava, parameterName=Uvspeed, sortPriority=5}), B=Time()) {constB=0.1}
  n35 = TextureSampleParameter2D(UVs=NamedRerouteDeclaration(None=Add(A=@n33, B=@n34)) {name=ScrollUv, nodeColor=<rgba 0 1 0.7656 1>}) {group=GoldenLava, parameterName=NormalMap, samplerType=SAMPLERTYPE_Normal, texture=&/T_MoltenGold_Normal}
  NamedRerouteDeclaration(None=MaterialFunctionCall(Normal=@n35.RGB, Flatness=ScalarParameter() {group=GoldenLava, parameterName=NormalMapIntensity}) {materialFunction=&FlattenNormal}.Result) {name=GoldenNormal, nodeColor=<rgba 0 0.5312 1 1>}
  TextureSample(UVs=NamedRerouteUsage(), Tex=TextureObject() {samplerType=SAMPLERTYPE_Masks, texture=&LavaGround_M}) {samplerType=SAMPLERTYPE_Masks, texture=&LavaGround_M}
 ]
}
